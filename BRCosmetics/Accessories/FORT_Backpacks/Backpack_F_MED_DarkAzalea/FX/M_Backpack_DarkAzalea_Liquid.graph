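# Masked, lit material for the liquid in the Dark Azalea backpack cosmetic, set up for skeletal meshes.
# Glow combines two panning smoke samples, one noise-distorted, with inside/outside Fresnel colors, top-facing tint and a UV edge band.
# A UV V split at 0.5 masks the smoke, weights fake HDR-sky reflections and confines scrolling noise-plus-bubble normals to one half.
# A time-of-day color function tints both base color and emissive.
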Material /BRCosmetics/Accessories/FORT_Backpacks/Backpack_F_MED_DarkAzalea/FX/M_Backpack_DarkAzalea_Liquid
graph {
 properties {
  bUsedWithMeshDeformer=true
  bUsedWithSkeletalMesh=true
  blendMode=BLEND_Masked
  pixelDepthOffsetMode=PDOM_Legacy
  translucencyLightingMode=TLM_Surface
  translucencyPass=MTP_BeforeDOF
  usedShadingModels=Default Lit
 }
 nodes [
  Constant() {r=1}
  n1 = MaterialFunctionCall(
   Color_Morning=VectorParameter() {parameterName="Color_Morning (V3)", defaultValue=<rgba 0.45 1 0.3694 0>}.RGB
   Color_Day=Constant() {r=1}
   Color_Evening=VectorParameter() {parameterName="Color_Evening (V3)", defaultValue=<rgba 0.7234 1 0.427 0>}.RGB
   Color_Night=VectorParameter() {parameterName="Color_Night (V3)", defaultValue=<rgba 1.3 1.3 1.3 0>}.RGB
  ) {materialFunction=&Color_Time_of_Day}
  n2 = TextureCoordinate()
  n3 = Subtract(A=OneMinus(None=ComponentMask(None=@n2) {g=true})) {constB=0.93}
  n4 = Multiply(A=VectorParameter() {defaultValue=<rgba 0.0621 0.1094 0.0103 0>}.RGB, B=Saturate(None=Multiply(A=@n3) {constB=1000}))
  n5 = DotProduct(A=VertexNormalWS(), B=Constant3Vector() {constant=<rgba 0 0 1 0>})
  n6 = OneMinus(None=Fresnel() {baseReflectFraction=0, exponent=1})
  n7 = LinearInterpolate(
   A=VectorParameter() {parameterName=Emissive Color outside, defaultValue=<rgba 0.025 0.05 0 0>}.RGB
   B=VectorParameter() {parameterName=Emissive Color inside, defaultValue=<rgba 0.1 0.2 0 0>}.RGB
   Alpha=Saturate(None=Power(Base=@n6) {constExponent=5})
  )
  n8 = TextureCoordinate() {uTiling=5, vTiling=5}
  n9 = TextureSample(UVs=Panner(Coordinate=Add(A=@n8) {constB=0}) {speedY=-0.2}) {texture=&T_Inky_Smoke_Tile}
  n10 = TextureCoordinate() {uTiling=3, vTiling=3}
  n11 = TextureSample(UVs=Panner(Coordinate=TextureCoordinate() {uTiling=4, vTiling=4}) {speedY=-0.2}) {samplerType=SAMPLERTYPE_Normal, texture=/Game/Characters/Player/Female/Medium/Bodies/F_MED_Bistro_Astronaut/FX/T_GaussianNoise_Normal}
  n12 = Multiply(A=@n11.R) {constB=0.1}
  n13 = TextureSample(UVs=Panner(Coordinate=Add(A=@n10, B=@n12)) {speedY=0.1}) {texture=&T_Inky_Smoke_Tile}
  n14 = Power(Base=Multiply(A=Multiply(A=@n9.R, B=@n13.R)) {constB=7}) {constExponent=1}
  n15 = TextureCoordinate()
  n16 = Power(Base=OneMinus(None=ComponentMask(None=@n15) {g=true})) {constExponent=22}
  n17 = ComponentMask(None=TextureCoordinate()) {g=true}
  n18 = Ceil(None=Subtract(A=OneMinus(None=@n17)) {constB=0.5})
  n19 = Add(A=Saturate(None=Multiply(A=@n16) {constB=20}), B=OneMinus(None=Saturate(None=@n18)))
  n20 = Add(
   A=Multiply(A=Saturate(None=@n5), B=VectorParameter() {parameterName=Color03, defaultValue=<rgba 0.0255 0.04 0 0>}.RGB) {constB=0.2}
   B=Multiply(A=Multiply(A=@n7, B=@n14), B=Saturate(None=@n19))
  )
  MaterialFunctionCall(Emissive=Multiply(A=@n1.Result, B=Add(A=@n4, B=@n20))) {materialFunction=&MF_FX_EmissiveAdjustments}
  n21 = ReflectionVectorWS()
  n22 = OneMinus(None=ComponentMask(None=TextureCoordinate()) {g=true})
  n23 = Saturate(None=Ceil(None=Subtract(A=@n22) {constB=0.5}))
  n24 = Multiply(A=Multiply(A=TextureSample(UVs=@n21) {texture=/Game/Packages/Fortress_Sky/TexturesHDR/T_hdrsky_test03_low}.RGB, B=Saturate(None=@n23))) {constB=0.3}
  n25 = ReflectionVectorWS()
  n26 = Multiply(A=Multiply(A=TextureSample(UVs=@n25) {texture=&T_hdrsky_test03_low}.RGB, B=OneMinus(None=@n23))) {constB=0.2}
  n27 = VectorParameter() {parameterName=Base Color outside, defaultValue=<rgba 0.0781 0.1562 0 0>}
  n28 = VectorParameter() {parameterName=Base Color inside, defaultValue=<rgba 0.2292 0.4583 0 0>}
  n29 = Fresnel() {baseReflectFraction=0, exponent=1}
  n30 = Saturate(None=Power(Base=OneMinus(None=@n29)) {constExponent=5})
  Multiply(A=@n1.Result, B=Add(A=Add(A=@n24, B=@n26), B=LinearInterpolate(A=@n27.RGB, B=@n28.RGB, Alpha=@n30)))
  n31 = OneMinus(None=ComponentMask(None=TextureCoordinate()) {g=true})
  n32 = Saturate(None=Ceil(None=Subtract(A=@n31) {constB=0.5}))
  n33 = TextureSample(UVs=Panner(Coordinate=TextureCoordinate() {uTiling=2.5, vTiling=2.5}) {speedY=0.12}) {samplerType=SAMPLERTYPE_Normal, texture=&T_GaussianNoise_Normal}
  n34 = Panner(Coordinate=TextureCoordinate() {uTiling=1.5, vTiling=1.5}) {speedY=-0.07}
  n35 = MaterialFunctionCall(
   BaseNormal=@n33.RGB
   AdditionalNormal=MaterialFunctionCall(
    Normal=TextureSample(UVs=@n34) {samplerType=SAMPLERTYPE_Normal, texture=/Game/Global/Textures/Default/Noise/T_Bubble_B_N}.RGB
    Flatness=ScalarParameter() {defaultValue=0.5, parameterName="Flatness (S)_1"}
   ) {materialFunction=&FlattenNormal}.Result
  ) {materialFunction=&BlendAngleCorrectedNormals}
  n36 = ScalarParameter() {parameterName="Flatness (S)_2"}
  Reroute(None=LinearInterpolate(
    A=Constant3Vector() {constant=<rgba 0 0 1 0>}
    B=MaterialFunctionCall(Normal=@n35.Result, Flatness=@n36) {materialFunction=&FlattenNormal}.Result
    Alpha=OneMinus(None=@n32)
   ))
  n37 = Power(Base=OneMinus(None=Fresnel() {baseReflectFraction=0, exponent=1})) {constExponent=3}
  Reroute(None=LinearInterpolate(Alpha=Saturate(None=@n37)) {constA=0.8, constB=0.9})
  ScalarParameter() {defaultValue=0.1, parameterName=Metallic}
  ScalarParameter() {defaultValue=0.2, parameterName=Specular}
  ScalarParameter() {defaultValue=0.25, parameterName=Roughness}
 ]
}
