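# Niagara system for the Dark Azalea backpack cosmetic that emits small, short-lived, green bubble sprites.
# Spawn rate is gated by the Use Loop Bubble user bool and scaled by a system-level interpolated moving average.
# Bubbles spawn in an offset box, drift upward with drag and grow over life.
# An inverted sphere kill volume contains them, with its origin adjusted between gameplay and front end.
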NiagaraSystem /BRCosmetics/Accessories/FORT_Backpacks/Backpack_F_MED_DarkAzalea/FX/NS_Backpack_DarkAzalea_Bubbles
dependencies {
 usedDataInterfaces [&NiagaraDataInterfaceCurve]
 usedDynamicInputs [&SetBoolByFloatComparison, &VectorLength]
 usedModules [&SystemState, &InterpolateOverTime]
}
emitters {bubbles {
  emitterUpdateScript [
   &EmitterState()
   &SpawnRate(
    SpawnRate=&Multiply_Float(
     A=&MakeCustomFloatFromBool(Bool=$User.Use Loop Bubble, True Float=200, False Float=0)
     B=$System.InterpolateOverTime.InterpolateOverTime_Float.Moving Average
    )
    Use Spawn Probability=false
   )
  ]
  particleSpawnScript [
   &InitializeParticle(
    Color=<rgba 0.4902 1 0.2001 1>
    Lifetime Max=0.5
    Lifetime Min=0.3
    Lifetime Mode=::NewEnumerator1
    Position Offset=<0 -15 -8>
    Sprite Size Mode=::NewEnumerator1
    Uniform Sprite Size Max=1.2
    Uniform Sprite Size Min=0.5
    UsePositionOffset=true
    Write Material Random=true
   )
   &AddVelocity(Velocity=<0 0 10>)
   &ShapeLocation(Box Size=<20 20 5>, MaskSpawnGroup=false, Override Seed=false, Shape Primitive=::NewEnumerator2, Surface Only=false, UseOwnerScale=true)
  ]
  particleUpdateScript [
   &ParticleState()
   &ScaleSpriteSize(Uniform Curve Sprite Scale=DataInterface() {propertyValues {Curve {keys [{interpMode=RCIM_Cubic, tangentMode=RCTM_User, time=0, value=0}, {interpMode=RCIM_Cubic, tangentMode=RCTM_User, time=1, value=1}]}}})
   &Drag(Use Linear Drag=true, Use Rotational Drag=true)
   &SolveForcesAndVelocity(Clamp Velocity=false, Limit Acceleration=false)
   &KillParticlesInVolume(
    Invert Volume=true
    Origin Offset:Vector3f=&FN_FrontEndVector3Modifier(Gameplay=<0 -15 -2>, Front End=<0 -15 -3>)
    Sphere Radius=11
   )
  ]
  rendererClasses [&NiagaraSpriteRendererProperties]
  renderers [{rendererClass=&NiagaraSpriteRendererProperties}]
 }}
summary {
 systemName=NS_Backpack_DarkAzalea_Bubbles
 emitters [{emitterName=bubbles, rendererClasses [&NiagaraSpriteRendererProperties]}]
 userVariables [
  User.bIsFrontEnd {type=NiagaraBool}
  User.bIsFrontEndPreview {type=NiagaraBool}
  User.Use Loop Bubble {type=NiagaraBool}
 ]
}
