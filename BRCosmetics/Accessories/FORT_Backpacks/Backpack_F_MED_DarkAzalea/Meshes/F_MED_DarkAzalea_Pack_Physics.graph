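# Physics asset for the Dark Azalea backpack cosmetic, giving simple collision and joints to the pack and its glass piece.
# Three capsule bodies cover the pack, a tiny glass_Master bone and a larger, downward-offset dyn_glass bone, which the name suggests is the dynamic part.
# Two constraints chain them: glass_Master rotates freely against pack, while dyn_glass has asymmetric angular limit values against glass_Master, with one swing limit at zero.
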PhysicsAsset /BRCosmetics/Accessories/FORT_Backpacks/Backpack_F_MED_DarkAzalea/Meshes/F_MED_DarkAzalea_Pack_Physics
bodies {
 dyn_glass [{shapeType=Capsule, length=8.8318, radius=7.5, center=<0 0 -20>}]
 glass_Master [{shapeType=Capsule, length=0, radius=0.5064, shapeName=glass_Master_capsule, rotation=<rot 0 -90 0>}]
 pack [{shapeType=Capsule, length=0.375, radius=4.8982}]
}
constraints [
 {bone1Name=dyn_glass, bone2Name=glass_Master, swing1LimitDegrees=0, swing2LimitDegrees=90, twistLimitDegrees=90}
 {
  bone1Name=glass_Master
  bone2Name=pack
  swing1LimitDegrees=45
  swing1Motion=Free
  swing2LimitDegrees=45
  swing2Motion=Free
  twistLimitDegrees=45
  twistMotion=Free
 }
]
