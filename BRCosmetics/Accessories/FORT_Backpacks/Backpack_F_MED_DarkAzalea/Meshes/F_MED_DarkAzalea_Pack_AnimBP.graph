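# Animation Blueprint for the Dark Azalea backpack cosmetic, animating its glass and liquid as the player moves.
# AnimGraph scales the pack bone, then chains spring AnimDynamics on water bones, a glass tilt and RigidBody simulation relative to the glass.
# A world-space rotation replace holds the liquid master bone's orientation before an extra dive tilt.
# EventGraph sets tilt and physics alphas from jump, crouch-sprint and skydive states, smoothed by interpolated alpha clamps.
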AnimBlueprint /BRCosmetics/Accessories/FORT_Backpacks/Backpack_F_MED_DarkAzalea/Meshes/F_MED_DarkAzalea_Pack_AnimBP
target_skeleton=&/F_MED_DarkAzalea_Pack_Skeleton
graphs [
 AnimGraph {nodes [
   n1 = "Transform (Modify) Bone - Bone: pack"(ComponentPose=Local To Component(LocalPose=Input Pose()), Translation="0,0,0", Rotation="0,0,0", Scale="0.95,0.95,0.95") {scaleMode=BMM_Additive, scaleSpace=BCS_BoneSpace, alphaBoolBlend {blendOption=Linear}, boneToModify {boneName=pack}, scale=<0.95 0.95 0.95>}
   n2 = Get jumpReactivity() -> (jumpReactivity:double)
   n3 = Anim Dynamics - Bone: dyn_waterDynamics_1(ComponentPose=@n1, Alpha=@n2, ExternalForce="0,0,0") {
    angularSpringConstant=500
    bAngularSpring=true
    bGravityOverrideInSimSpace=true
    gravityScale=0
    alphaBoolBlend {blendOption=Linear}
    alphaScaleBiasClamp {bInterpResult=true, interpSpeedDecreasing=20, interpSpeedIncreasing=20, scale=0.5}
    boundBone {boneName=dyn_waterDynamics_1}
    componentLinearAccScale=<0.9 0.9 0.9>
    physicsBodyDefinitions [{
      sphereCollisionRadius=0
      boundBone {boneName=dyn_waterDynamics_1}
      boxExtents=<5 5 5>
      constraintSetup {angularLimitsMax=<15 30 30>, angularLimitsMin=<-15 -2 -30>, angularTarget=<1 0 0>}
      localJointOffset=<0 -0.2 0.5>
     }]
   }
   n4 = Anim Dynamics - Bone: dyn_waterDynamics_2(ComponentPose=@n3, Alpha=@n2, ExternalForce="0,0,0") {
    angularSpringConstant=500
    bAngularSpring=true
    bGravityOverrideInSimSpace=true
    gravityScale=0
    alphaBoolBlend {blendOption=Linear}
    alphaScaleBiasClamp {bInterpResult=true, interpSpeedDecreasing=20, interpSpeedIncreasing=20, scale=0.5}
    boundBone {boneName=dyn_waterDynamics_2}
    componentLinearAccScale=<0.9 0.9 0.9>
    physicsBodyDefinitions [{
      sphereCollisionRadius=0
      boundBone {boneName=dyn_waterDynamics_2}
      boxExtents=<5 5 5>
      constraintSetup {angularLimitsMax=<15 15 15>, angularLimitsMin=<-15 -2 -15>, angularTarget=<1 0 0>}
      localJointOffset=<0 0.2 0.5>
     }]
   }
   n5 = Anim Dynamics - Bone: dyn_waterDynamics_3(ComponentPose=@n4, Alpha=1, ExternalForce="0,0,0") {
    angularSpringConstant=50
    bAngularSpring=true
    bGravityOverrideInSimSpace=true
    bLinearSpring=true
    linearSpringConstant=200
    simulationSpace=BoneRelative
    alphaBoolBlend {blendOption=Linear}
    boundBone {boneName=dyn_waterDynamics_3}
    componentLinearAccScale=<1.5 1.5 1.15>
    physicsBodyDefinitions [{
      sphereCollisionRadius=0
      boundBone {boneName=dyn_waterDynamics_3}
      boxExtents=<5 5 5>
      constraintSetup {angularLimitsMax=<10 10 10>, angularLimitsMin=<-10 -10 -10>, angularTarget=<1 0 0>, linearAxesMax=<5 0 0>}
      localJointOffset=<0 0 0>
     }]
   }
   n6 = "Transform (Modify) Bone - Bone: glass_Master"(
    ComponentPose=@n5
    Alpha="Get FInterp(Alpha, (1: 1))"() -> ("FInterp(Alpha, (1: 1))":float)
    Translation="0,0,0"
    Rotation="0,0,-90"
    Scale="1,1,1"
   ) {
    rotationMode=BMM_Additive
    rotationSpace=BCS_BoneSpace
    translationMode=BMM_Additive
    translationSpace=BCS_BoneSpace
    alphaBoolBlend {blendOption=Linear}
    alphaScaleBiasClamp {bInterpResult=true, interpSpeedDecreasing=3, interpSpeedIncreasing=3}
    boneToModify {boneName=glass_Master}
    rotation {roll=-90}
   }
   n7 = RigidBody(ComponentPose=@n6, Alpha=Get Alpha() -> (Alpha:float)) {
    bOverrideWorldGravity=true
    overridePhysicsAsset=&/F_MED_DarkAzalea_Pack_Physics
    simulationSpace=BaseBoneSpace
    alphaBoolBlend {blendOption=Linear}
    baseBoneRef {boneName=glass_Master}
    componentAppliedLinearAccClamp=<3000 3000 3000>
    componentLinearAccScale=<0.3 0.3 0.3>
    componentLinearVelScale=<0.3 0.3 0.3>
    overrideWorldGravity {z=-980}
    simSpaceSettings {velocityScaleZ=0.2}
   }
   n8 = "Transform (Modify) Bone - Bone: dyn_liquid_master"(ComponentPose=@n7, Alpha=1, Translation="0,0,0", Rotation="0,0,0", Scale="1,1,1") {
    rotationMode=BMM_Replace
    rotationSpace=BCS_WorldSpace
    translationMode=BMM_Additive
    translationSpace=BCS_BoneSpace
    alphaBoolBlend {blendOption=Linear}
    alphaScaleBiasClamp {bInterpResult=true, interpSpeedDecreasing=6, interpSpeedIncreasing=6}
    boneToModify {boneName=dyn_liquid_master}
   }
   n9 = "Transform (Modify) Bone - Bone: glass_Master"(ComponentPose=@n8, Alpha="To Float (Boolean)"(InBool=Get bIsSkydiveDiveMode()), Translation="0,0,0", Rotation="0,0,-50", Scale="1,1,1") {
    rotationMode=BMM_Additive
    rotationSpace=BCS_BoneSpace
    translationMode=BMM_Additive
    translationSpace=BCS_BoneSpace
    alphaBoolBlend {blendOption=Linear}
    alphaScaleBiasClamp {bInterpResult=true, interpSpeedDecreasing=6, interpSpeedIncreasing=6}
    boneToModify {boneName=glass_Master}
    rotation {roll=-50}
   }
   Output Pose(Result=Component To Local(ComponentPose=@n9)) {layerGroup=None}
  ]}
 EventGraph {nodes [
   Event BlueprintUpdateAnimation()
   > n1 = Sequence!(execute) -> (then_0, then_1, then_2:Exec, then_3:Exec)
   > Set jumpReactivity(jumpReactivity:double=SelectFloat(A=0.4, B=0.6, bPickA=Get bIsJumping())) -> (Output_Get:double)
   n2 = Branch(execute=@n1.then_1, Condition=Get bIsCrouchSprinting())
   n3 = "Get FInterp(Alpha, (1: 1))_0"() -> ("FInterp(Alpha, (1: 1))_0":float)
   > "Set FInterp(Alpha, (1: 1))"("FInterp(Alpha, (1: 1))":float=@n3) -> (Output_Get:float)
   n4 = Branch(execute=@n2.else, Condition=Get bIsSkydiveFloating())
   > "Set FInterp(Alpha, (1: 1))"("FInterp(Alpha, (1: 1))"=@n3)
   n5 = Branch(execute=@n4.else, Condition=Get bIsSkydiveDiveMode())
   > "Set FInterp(Alpha, (1: 1))"("FInterp(Alpha, (1: 1))"=@n3)
   "Set FInterp(Alpha, (1: 1))"(execute=@n5.else, "FInterp(Alpha, (1: 1))"=0)
   n6 = Branch(execute=@n1.then_2, Condition=Get bIsSkyDiving())
   > Set Alpha(Alpha:float=Get Alpha_SkyDiving() -> (Alpha_SkyDiving:float)) -> (Output_Get:float)
   Set Alpha(execute=@n6.else, Alpha=1)
   n7 = Branch(execute=@n1.then_3, Condition=Get bIsSkydiveDiveMode())
   > Set Rotation(Rotation:Rotator=Get Rotation_0() -> (Rotation_0:Rotator)) -> (Output_Get:Rotator)
   Set Rotation(execute=@n7.else, Rotation=Get Rotation_1() -> (Rotation_1:Rotator))
   TryGetPawnOwner()
  ]}
]
variables [
 Alpha {type=number, default=1}
 Alpha_SkyDiving {type=number, default=0.7}
 "FInterp(Alpha, (1: 1))" {type=number}
 "FInterp(Alpha, (1: 1))_0" {type=number, default=1}
 jumpReactivity {type=number}
 Rotation {type=Rotator}
 Rotation_0 {type=Rotator, default=<rot 0 0 -180>}
 Rotation_1 {type=Rotator, default=<rot 0 0 -90>}
]
